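# Physics asset for the back skirt of the F_MED_RemoteControl player body, providing bone physics for its dynamic skirt chains.
# Left and right three-bone dyn_back capsule chains hang from small MSTR_back anchor bodies through limited swing and twist constraints.
# A cross constraint with free swing and twist ties the two chain tips together, coupling both sides.
# Pelvis, thigh, calf and foot capsules carry no constraints and provide lower-body collision shapes.
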PhysicsAsset /BRCosmetics/Characters/Player/Female/Medium/Bodies/F_MED_RemoteControl/Meshes/F_MED_RemoteControl_Physics_BackSkirt
bodies {
 MSTR_back_l [{shapeType=Capsule, length=2, radius=1, shapeName=MSTR_back_l_capsule}]
 MSTR_back_r [{shapeType=Capsule, length=2, radius=1, shapeName=MSTR_back_r_capsule}]
 calf_l [{shapeType=Capsule, length=26.5, radius=15.5, shapeName=calf_l_capsule, center=<-21.1214 -2.024 0>, rotation=<rot 90 0 0>}]
 calf_r [{shapeType=Capsule, length=26.5, radius=15.5, shapeName=calf_r_capsule, center=<22.5895 2.024 0>, rotation=<rot 90 180 0>}]
 dyn_back_l_01 [{shapeType=Capsule, length=8.5, radius=12.5, shapeName=dyn_back_l_01_capsule, center=<8.6 -11.2 0>, rotation=<rot 90 0 0>}]
 dyn_back_l_02 [{shapeType=Capsule, length=6, radius=12.5, shapeName=dyn_back_l_02_capsule, center=<7.7 -11.2 0>, rotation=<rot 90 0 0>}]
 dyn_back_l_03 [{shapeType=Capsule, length=9.5, radius=14, shapeName=dyn_back_l_03_capsule, center=<7.65 -12 0>, rotation=<rot 90 180 -180>}]
 dyn_back_r_01 [{shapeType=Capsule, length=8.5, radius=12.5, shapeName=dyn_back_r_01_capsule, center=<8.6 -11.2 0>, rotation=<rot 90 0 0>}]
 dyn_back_r_02 [{shapeType=Capsule, length=6, radius=12.5, shapeName=dyn_back_r_02_capsule, center=<7.7 -11.2 0>, rotation=<rot 90 0 0>}]
 dyn_back_r_03 [{shapeType=Capsule, length=9.5, radius=14, shapeName=dyn_back_l_03_capsule, center=<7.65 -12 0>, rotation=<rot 90 180 -180>}]
 foot_l [{shapeType=Capsule, length=12.5, radius=12, shapeName=foot_l_capsule, center=<-2.3 5.6 0>, rotation=<rot 0 0 90>}]
 foot_r [{shapeType=Capsule, length=12, radius=12, shapeName=foot_r_capsule, center=<2.3 -5.6 0>, rotation=<rot 0 0 90>}]
 pelvis [{shapeType=Capsule, length=14.5, radius=10.5, shapeName=pelvis_capsule, center=<-6.8251 0 0>}]
 thigh_l [{shapeType=Capsule, length=31, radius=14, shapeName=thigh_l_capsule, center=<-30 -2 0>, rotation=<rot 90 89.4328 -95.1921>}]
 thigh_r [{shapeType=Capsule, length=31, radius=14, shapeName=thigh_r_capsule, center=<30 2 0>, rotation=<rot 90 89.4327 84.8079>}]
}
constraints [
 {bone1Name=dyn_back_l_03, bone2Name=dyn_back_l_02, swing1LimitDegrees=20, swing2LimitDegrees=20, twistLimitDegrees=1}
 {bone1Name=dyn_back_l_02, bone2Name=dyn_back_l_01, swing1LimitDegrees=30, swing2LimitDegrees=20, twistLimitDegrees=20}
 {bone1Name=dyn_back_l_01, bone2Name=MSTR_back_l, swing1LimitDegrees=30, swing2LimitDegrees=10, twistLimitDegrees=1}
 {bone1Name=dyn_back_r_03, bone2Name=dyn_back_r_02, swing1LimitDegrees=20, swing2LimitDegrees=20, twistLimitDegrees=1}
 {bone1Name=dyn_back_r_02, bone2Name=dyn_back_r_01, swing1LimitDegrees=30, swing2LimitDegrees=20, twistLimitDegrees=20}
 {bone1Name=dyn_back_r_01, bone2Name=MSTR_back_r, swing1LimitDegrees=30, swing2LimitDegrees=10, twistLimitDegrees=1}
 {
  bone1Name=dyn_back_r_03
  bone2Name=dyn_back_l_03
  swing1LimitDegrees=45
  swing1Motion=Free
  swing2LimitDegrees=45
  swing2Motion=Free
  twistLimitDegrees=45
  twistMotion=Free
 }
]
