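# Physics asset for the front skirt of the F_MED_RemoteControl female cosmetic body, a crouch variant going by its name.
# Left and right chains of five dynamic capsules hang from MSTR bones constrained to the pelvis, with narrow swing and zero twist limits.
# Pelvis, root, thigh and calf capsules form the body shapes, joined by free constraints.
# The right chain alone adds free constraints from its last bone to the other four.
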PhysicsAsset /BRCosmetics/Characters/Player/Female/Medium/Bodies/F_MED_RemoteControl/Meshes/F_MED_RemoteControl_Physics_FrontSkirt_CrouchNew
bodies {
 MSTR_front_l_01 [{shapeType=Capsule, length=0.5, radius=0.5, shapeName=MSTR_front_l_01_capsule}]
 MSTR_front_r_01 [{shapeType=Capsule, length=0.5, radius=0.5, shapeName=MSTR_front_r_01_capsule}]
 calf_l [{shapeType=Capsule, length=33, radius=13, shapeName=calf_l_capsule, center=<-20.3528 0 0>, rotation=<rot 90 80.7734 86.1019>}]
 calf_r [{shapeType=Capsule, length=33, radius=13, shapeName=calf_l_capsule, center=<20.3528 0 0>, rotation=<rot 90 145.4915 -29.18>}]
 dyn_front_l_01 [{shapeType=Capsule, length=12.5, radius=5, shapeName=dyn_front_l_01_capsule, center=<7.5558 -5 0>, rotation=<rot 0 0 -16>}]
 dyn_front_l_01_A [{shapeType=Capsule, length=13, radius=7.5, shapeName=dyn_front_l_01_A_capsule, center=<7.4849 -5 0>}]
 dyn_front_l_02 [{shapeType=Capsule, length=14.5, radius=6.5, shapeName=dyn_front_l_02_capsule, center=<5.6874 -5 0>}]
 dyn_front_l_03 [{shapeType=Capsule, length=13, radius=7.5, shapeName=dyn_front_l_03_capsule, center=<10.2614 -5 0>}]
 dyn_front_l_A [{shapeType=Capsule, length=9.5, radius=6.5, shapeName=dyn_front_l_A_capsule, center=<5.6874 -5 0>}]
 dyn_front_r_01 [{shapeType=Capsule, length=13.5, radius=5.5, shapeName=dyn_front_r_01_capsule, center=<7.5558 -6 0>, rotation=<rot 0 0 15>}]
 dyn_front_r_01_B [{shapeType=Capsule, length=11, radius=7.5, shapeName=dyn_front_r_01_B_capsule, center=<7.4849 -5 0>}]
 dyn_front_r_02 [{shapeType=Capsule, length=14, radius=7, shapeName=dyn_front_r_02_capsule, center=<5.6874 -5 0>}]
 dyn_front_r_03 [{shapeType=Capsule, length=11, radius=7.5, shapeName=dyn_front_r_03_capsule, center=<10.2614 -5 0>}]
 dyn_front_r_B [{shapeType=Capsule, length=9.5, radius=6.5, shapeName=dyn_front_r_B_capsule, center=<5.6874 -5 0>}]
 pelvis [{shapeType=Capsule, length=7.5, radius=11.5, shapeName=pelvis_capsule, center=<0 -1.471 0>}]
 root [{shapeType=Capsule, length=5, radius=5, shapeName=root_capsule}]
 thigh_l [{shapeType=Capsule, length=42, radius=13, shapeName=thigh_l_capsule, center=<-24.7727 -1.3709 1.6783>, rotation=<rot 90 180 -178.2996>}]
 thigh_r [{shapeType=Capsule, length=42, radius=13, shapeName=thigh_l_capsule, center=<24.7727 1.3709 -1.6783>, rotation=<rot 90 180 1.7004>}]
}
constraints [
 {
  bone1Name=pelvis
  bone2Name=root
  swing1LimitDegrees=45
  swing1Motion=Free
  swing2LimitDegrees=45
  swing2Motion=Free
  twistLimitDegrees=45
  twistMotion=Free
 }
 {
  bone1Name=MSTR_front_l_01
  bone2Name=pelvis
  swing1LimitDegrees=45
  swing1Motion=Free
  swing2LimitDegrees=45
  swing2Motion=Free
  twistLimitDegrees=45
  twistMotion=Free
 }
 {bone1Name=dyn_front_l_01, bone2Name=MSTR_front_l_01, swing1LimitDegrees=27, swing2LimitDegrees=5, twistLimitDegrees=0}
 {bone1Name=dyn_front_l_A, bone2Name=dyn_front_l_01, swing1LimitDegrees=27, swing2LimitDegrees=5, twistLimitDegrees=0}
 {bone1Name=dyn_front_l_02, bone2Name=dyn_front_l_A, swing1LimitDegrees=27, swing2LimitDegrees=5, twistLimitDegrees=0}
 {bone1Name=dyn_front_l_01_A, bone2Name=dyn_front_l_02, swing1LimitDegrees=27, swing2LimitDegrees=5, twistLimitDegrees=0}
 {bone1Name=dyn_front_l_03, bone2Name=dyn_front_l_01_A, swing1LimitDegrees=27, swing2LimitDegrees=5, twistLimitDegrees=0}
 {
  bone1Name=MSTR_front_r_01
  bone2Name=pelvis
  swing1LimitDegrees=45
  swing1Motion=Free
  swing2LimitDegrees=45
  swing2Motion=Free
  twistLimitDegrees=45
  twistMotion=Free
 }
 {bone1Name=dyn_front_r_01, bone2Name=MSTR_front_r_01, swing1LimitDegrees=27, swing2LimitDegrees=5, twistLimitDegrees=0}
 {bone1Name=dyn_front_r_B, bone2Name=dyn_front_r_01, swing1LimitDegrees=27, swing2LimitDegrees=5, twistLimitDegrees=0}
 {bone1Name=dyn_front_r_02, bone2Name=dyn_front_r_B, swing1LimitDegrees=27, swing2LimitDegrees=5, twistLimitDegrees=0}
 {bone1Name=dyn_front_r_01_B, bone2Name=dyn_front_r_02, swing1LimitDegrees=27, swing2LimitDegrees=5, twistLimitDegrees=0}
 {bone1Name=dyn_front_r_03, bone2Name=dyn_front_r_01_B, swing1LimitDegrees=27, swing2LimitDegrees=5, twistLimitDegrees=0}
 {
  bone1Name=thigh_l
  bone2Name=pelvis
  swing1LimitDegrees=45
  swing1Motion=Free
  swing2LimitDegrees=45
  swing2Motion=Free
  twistLimitDegrees=45
  twistMotion=Free
 }
 {
  bone1Name=calf_l
  bone2Name=thigh_l
  swing1LimitDegrees=45
  swing1Motion=Free
  swing2LimitDegrees=45
  swing2Motion=Free
  twistLimitDegrees=45
  twistMotion=Free
 }
 {
  bone1Name=calf_r
  bone2Name=thigh_r
  swing1LimitDegrees=45
  swing1Motion=Free
  swing2LimitDegrees=45
  swing2Motion=Free
  twistLimitDegrees=45
  twistMotion=Free
 }
 {
  bone1Name=thigh_r
  bone2Name=pelvis
  swing1LimitDegrees=45
  swing1Motion=Free
  swing2LimitDegrees=45
  swing2Motion=Free
  twistLimitDegrees=45
  twistMotion=Free
 }
 {
  bone1Name=dyn_front_r_01
  bone2Name=dyn_front_r_03
  swing1LimitDegrees=45
  swing1Motion=Free
  swing2LimitDegrees=45
  swing2Motion=Free
  twistLimitDegrees=45
  twistMotion=Free
 }
 {
  bone1Name=dyn_front_r_B
  bone2Name=dyn_front_r_03
  swing1LimitDegrees=45
  swing1Motion=Free
  swing2LimitDegrees=45
  swing2Motion=Free
  twistLimitDegrees=45
  twistMotion=Free
 }
 {
  bone1Name=dyn_front_r_02
  bone2Name=dyn_front_r_03
  swing1LimitDegrees=45
  swing1Motion=Free
  swing2LimitDegrees=45
  swing2Motion=Free
  twistLimitDegrees=45
  twistMotion=Free
 }
 {
  bone1Name=dyn_front_r_01_B
  bone2Name=dyn_front_r_03
  swing1LimitDegrees=45
  swing1Motion=Free
  swing2LimitDegrees=45
  swing2Motion=Free
  twistLimitDegrees=45
  twistMotion=Free
 }
]
