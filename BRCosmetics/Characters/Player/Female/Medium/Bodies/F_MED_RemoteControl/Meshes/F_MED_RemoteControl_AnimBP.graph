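# Animation Blueprint for the F_MED_RemoteControl female body cosmetic, adding secondary motion to its skirt-like cloth, pants and sleeves.
# Copies the parent mesh pose through a Control Rig, AnimDynamics pants and additive corrective bones.
# Blends to RigidBody skirt simulations when enabled, otherwise to a scalability fallback using constraints and posed offsets for crouch, slide, skydive and DBNO.
# The event graph derives blend alphas from movement state, speed and thigh rotation.
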AnimBlueprint /BRCosmetics/Characters/Player/Female/Medium/Bodies/F_MED_RemoteControl/Meshes/F_MED_RemoteControl_AnimBP
target_skeleton=&/F_MED_RemoteControl_Skeleton
graphs [
 AnimGraph {nodes [
   Output Pose(Result=Blend Poses by bool(
     BlendPose_0=Use cached pose 'RBANoutput'()
     BlendPose_1=Use cached pose 'SCALABILITY output'()
     BlendTime_0=0.3
     bActiveValue=Get bRigidBodyEnabled()
    ) {blendType=Linear}) {layerGroup=None}
   n1 = "Transform (Modify) Bone - Bone: deform_sleeve_r"(
    ComponentPose=Local To Component(LocalPose=Use cached pose 'LOWER PANTS'())
    Alpha=Get upperarm_r alpha() -> (upperarm_r alpha:float)
    Translation="10,0,20"
    Rotation="0,0,0"
    Scale="1,1,1"
   ) {
    rotationMode=BMM_Additive
    rotationSpace=BCS_BoneSpace
    translationMode=BMM_Additive
    translationSpace=BCS_BoneSpace
    alphaBoolBlend {blendInTime=0.2, blendOption=Linear, blendOutTime=0.2}
    boneToModify {boneName=deform_sleeve_r}
    translation {x=10, z=20}
   }
   n2 = "Transform (Modify) Bone - Bone: deform_sleeve_l"(ComponentPose=@n1, Alpha=Get upperarm_l alpha() -> (upperarm_l alpha:float), Translation="10,0,-20", Rotation="0,0,0", Scale="1,1,1") {
    rotationMode=BMM_Additive
    rotationSpace=BCS_BoneSpace
    translationMode=BMM_Additive
    translationSpace=BCS_BoneSpace
    alphaBoolBlend {blendInTime=0.2, blendOption=Linear, blendOutTime=0.2}
    boneToModify {boneName=deform_sleeve_l}
    translation {x=10, z=-20}
   }
   n3 = "Get  jump alpha"() -> (" jump alpha":double)
   n4 = "Transform (Modify) Bone - Bone: MSTR_front_r_01"(ComponentPose=@n2, Alpha=@n3, Translation="0,-10,0", Rotation="0,0,-10", Scale="1,1,1") {
    rotationMode=BMM_Additive
    rotationSpace=BCS_BoneSpace
    translationMode=BMM_Additive
    translationSpace=BCS_BoneSpace
    alphaBoolBlend {blendInTime=0.2, blendOption=Linear, blendOutTime=0.2}
    boneToModify {boneName=MSTR_front_r_01}
    rotation {roll=-10}
    translation {y=-10}
   }
   n5 = "Transform (Modify) Bone - Bone: MSTR_front_l_01"(ComponentPose=@n4, Alpha=@n3, Translation="6,-3,0", Rotation="0,0,0", Scale="1,1,1") {
    rotationMode=BMM_Additive
    rotationSpace=BCS_BoneSpace
    translationMode=BMM_Additive
    translationSpace=BCS_BoneSpace
    alphaBoolBlend {blendInTime=0.2, blendOption=Linear, blendOutTime=0.2}
    boneToModify {boneName=MSTR_front_l_01}
    translation=<6 -3>
   }
   n6 = Get bIsSliding()
   n7 = "Transform (Modify) Bone - Bone: MSTR_back_r"!(ComponentPose=@n5, bAlphaBoolEnabled:bool=@n6, Translation="0,0,0", Rotation="0,-110,0", Scale="1,1,1") -> (Pose) {
    alphaInputType=Bool
    rotationMode=BMM_Additive
    rotationSpace=BCS_BoneSpace
    translationMode=BMM_Additive
    translationSpace=BCS_BoneSpace
    alphaBoolBlend {blendOption=Linear}
    boneToModify {boneName=MSTR_back_r}
    rotation {yaw=-110}
   }
   n8 = "Transform (Modify) Bone - Bone: MSTR_back_l"!(ComponentPose=@n7, bAlphaBoolEnabled:bool=@n6, Translation="0,0,0", Rotation="0,-110,0", Scale="1,1,1") -> (Pose) {
    alphaInputType=Bool
    rotationMode=BMM_Additive
    rotationSpace=BCS_BoneSpace
    translationMode=BMM_Additive
    translationSpace=BCS_BoneSpace
    alphaBoolBlend {blendOption=Linear}
    boneToModify {boneName=MSTR_back_l}
    rotation {yaw=-110}
   }
   n9 = "Transform (Modify) Bone - Bone: deform_pants_l"(ComponentPose=@n8, Alpha=Get calf_l_alpha() -> (calf_l_alpha:float), Translation="0,-26,0", Rotation="0,0,0", Scale="1,1,1") {
    rotationMode=BMM_Additive
    rotationSpace=BCS_BoneSpace
    translationMode=BMM_Additive
    translationSpace=BCS_BoneSpace
    alphaBoolBlend {blendInTime=0.2, blendOption=Linear, blendOutTime=0.2}
    boneToModify {boneName=deform_pants_l}
    translation {y=-26}
   }
   n10 = "Transform (Modify) Bone - Bone: deform_pants_r"(ComponentPose=@n9, Alpha=Get calf_r_alpha() -> (calf_r_alpha:float), Translation="0,-26,0", Rotation="0,0,0", Scale="1,1,1") {
    rotationMode=BMM_Additive
    rotationSpace=BCS_BoneSpace
    translationMode=BMM_Additive
    translationSpace=BCS_BoneSpace
    alphaBoolBlend {blendInTime=0.2, blendOption=Linear, blendOutTime=0.2}
    boneToModify {boneName=deform_pants_r}
    translation {y=-26}
   }
   Save cached pose 'Pose'(Pose=Component To Local(ComponentPose=@n10)) {cacheName=Pose}
   n11 = Get bachcloth fwd() -> (bachcloth fwd:double)
   n12 = "Transform (Modify) Bone - Bone: MSTR_back_r"!(
    ComponentPose=Local To Component(LocalPose=Use cached pose 'Pose'())
    Alpha:float=@n11
    Translation="0,0,0"
    Rotation="0,-20,0"
    Scale="1,1,1"
   ) -> (Pose) {
    rotationMode=BMM_Additive
    rotationSpace=BCS_BoneSpace
    translationMode=BMM_Additive
    translationSpace=BCS_BoneSpace
    alphaBoolBlend {blendOption=Linear}
    boneToModify {boneName=MSTR_back_r}
    rotation {yaw=-20}
   }
   n13 = "Transform (Modify) Bone - Bone: MSTR_back_l"!(ComponentPose=@n12, Alpha:float=@n11, Translation="0,0,0", Rotation="0,-20,0", Scale="1,1,1") -> (Pose) {
    rotationMode=BMM_Additive
    rotationSpace=BCS_BoneSpace
    translationMode=BMM_Additive
    translationSpace=BCS_BoneSpace
    alphaBoolBlend {blendOption=Linear}
    boneToModify {boneName=MSTR_back_l}
    rotation {yaw=-20}
   }
   n14 = RigidBody!(
    ComponentPose=@n13
    Alpha=Get FrontCloth Rigid alpha() -> (FrontCloth Rigid alpha:double)
    OverrideWorldGravity:Vector=Get Gravity Override()
    ExternalForce
    ComponentLinearAccScale:Vector="0.6,0.6,0.3"
    ComponentLinearVelScale:Vector="0.6,0.6,0.3"
   ) -> (Pose) {
    bClampLinearTranslationLimitToRefPose=true
    bOverrideWorldGravity=true
    overridePhysicsAsset=&/F_MED_RemoteControl_Physics_BackSkirt
    simulationSpace=BaseBoneSpace
    alphaBoolBlend {blendOption=Linear}
    baseBoneRef {boneName=pelvis}
    componentAppliedLinearAccClamp=<3000 3000 3000>
    componentLinearAccScale=<0.4 0.4 0.4>
    componentLinearVelScale=<0.6 0.6 0.6>
    overrideWorldGravity {z=-980}
   }
   Save cached pose 'BackRBANCache'(Pose=Component To Local(ComponentPose=@n14)) {cacheName=BackRBANCache}
   n15 = "Transform (Modify) Bone - Bone: dyn_back_r_01"!(
    ComponentPose=Local To Component(LocalPose=Use cached pose 'ALT CONSTRAINT'())
    bAlphaBoolEnabled:bool=True
    Translation="0,0,0"
    Rotation="0,-110,0"
    Scale="1,1,1"
   ) -> (Pose) {
    alphaInputType=Bool
    rotationMode=BMM_Additive
    rotationSpace=BCS_BoneSpace
    translationMode=BMM_Additive
    translationSpace=BCS_BoneSpace
    alphaBoolBlend {blendOption=Linear}
    boneToModify {boneName=dyn_back_r_01}
    rotation {yaw=-110}
   }
   n16 = "Transform (Modify) Bone - Bone: dyn_back_l_01"!(ComponentPose=@n15, bAlphaBoolEnabled:bool=True, Translation="0,0,0", Rotation="0,-110,0", Scale="1,1,1") -> (Pose) {
    alphaInputType=Bool
    rotationMode=BMM_Additive
    rotationSpace=BCS_BoneSpace
    translationMode=BMM_Additive
    translationSpace=BCS_BoneSpace
    alphaBoolBlend {blendOption=Linear}
    boneToModify {boneName=dyn_back_l_01}
    rotation {yaw=-110}
   }
   Save cached pose 'SLIDING FIX'(Pose=Blend Poses by bool(
     BlendPose_0=Component To Local(ComponentPose=@n16)
     BlendPose_1=Use cached pose 'ALT CONSTRAINT'()
     BlendTime_0=0.3
     bActiveValue=Get bIsSliding()
    ) {blendType=Linear}) {cacheName=SLIDING FIX}
   n17 = "Transform (Modify) Bone - Bone: dyn_back_r_01"!(
    ComponentPose=Local To Component(LocalPose=Use cached pose 'SCALABILITY input'())
    Alpha:float=1
    Translation="0,0,0"
    Rotation="0,-5,0"
    Scale="1,1,1"
   ) -> (Pose) {
    rotationMode=BMM_Additive
    rotationSpace=BCS_ParentBoneSpace
    translationMode=BMM_Additive
    translationSpace=BCS_BoneSpace
    alphaBoolBlend {blendOption=Linear}
    boneToModify {boneName=dyn_back_r_01}
    rotation {yaw=-5}
   }
   n18 = "Transform (Modify) Bone - Bone: dyn_back_l_01"!(ComponentPose=@n17, Alpha:float=1, Translation="0,0,0", Rotation="0,-5,0", Scale="1,1,1") -> (Pose) {
    rotationMode=BMM_Additive
    rotationSpace=BCS_BoneSpace
    translationMode=BMM_Additive
    translationSpace=BCS_BoneSpace
    alphaBoolBlend {blendOption=Linear}
    boneToModify {boneName=dyn_back_l_01}
    rotation {yaw=-5}
   }
   n19 = Get Alpha_Thigh_L_Back() -> (Alpha_Thigh_L_Back:double)
   n20 = "Transform (Modify) Bone - Bone: dyn_back_r_01"(ComponentPose=@n18, Alpha=@n19, Translation="0,0,0", Rotation="0,-15,0", Scale="1,1,1") {
    rotationMode=BMM_Additive
    rotationSpace=BCS_ParentBoneSpace
    translationMode=BMM_Additive
    translationSpace=BCS_BoneSpace
    alphaBoolBlend {blendOption=Linear}
    boneToModify {boneName=dyn_back_r_01}
    rotation {yaw=-15}
   }
   n21 = "Transform (Modify) Bone - Bone: dyn_back_r_02"(ComponentPose=@n20, Alpha=@n19, Translation="0,0,0", Rotation="0,-25,0", Scale="1,1,1") {
    rotationMode=BMM_Additive
    rotationSpace=BCS_ParentBoneSpace
    translationMode=BMM_Additive
    translationSpace=BCS_BoneSpace
    alphaBoolBlend {blendOption=Linear}
    boneToModify {boneName=dyn_back_r_02}
    rotation {yaw=-25}
   }
   n22 = "Transform (Modify) Bone - Bone: dyn_back_r_03"(ComponentPose=@n21, Alpha=@n19, Translation="0,0,0", Rotation="0,-10,0", Scale="1,1,1") {
    rotationMode=BMM_Additive
    rotationSpace=BCS_ParentBoneSpace
    translationMode=BMM_Additive
    translationSpace=BCS_BoneSpace
    alphaBoolBlend {blendOption=Linear}
    boneToModify {boneName=dyn_back_r_03}
    rotation {yaw=-10}
   }
   n23 = "Transform (Modify) Bone - Bone: dyn_back_l_01"(ComponentPose=@n22, Alpha=@n19, Translation="0,0,0", Rotation="0,-15,0", Scale="1,1,1") {
    rotationMode=BMM_Additive
    rotationSpace=BCS_BoneSpace
    translationMode=BMM_Additive
    translationSpace=BCS_BoneSpace
    alphaBoolBlend {blendOption=Linear}
    boneToModify {boneName=dyn_back_l_01}
    rotation {yaw=-15}
   }
   n24 = "Transform (Modify) Bone - Bone: dyn_back_l_02"(ComponentPose=@n23, Alpha=@n19, Translation="0,0,0", Rotation="0,-25,0", Scale="1,1,1") {
    rotationMode=BMM_Additive
    rotationSpace=BCS_BoneSpace
    translationMode=BMM_Additive
    translationSpace=BCS_BoneSpace
    alphaBoolBlend {blendOption=Linear}
    boneToModify {boneName=dyn_back_l_02}
    rotation {yaw=-25}
   }
   n25 = "Transform (Modify) Bone - Bone: dyn_back_l_03"(ComponentPose=@n24, Alpha=@n19, Translation="0,0,0", Rotation="0,-10,0", Scale="1,1,1") {
    rotationMode=BMM_Additive
    rotationSpace=BCS_BoneSpace
    translationMode=BMM_Additive
    translationSpace=BCS_BoneSpace
    alphaBoolBlend {blendOption=Linear}
    boneToModify {boneName=dyn_back_l_03}
    rotation {yaw=-10}
   }
   n26 = Get Alpha_Thigh_R_Back() -> (Alpha_Thigh_R_Back:double)
   n27 = "Transform (Modify) Bone - Bone: dyn_back_l_01"(ComponentPose=@n25, Alpha=@n26, Translation="0,0,0", Rotation="0,-15,0", Scale="1,1,1") {
    rotationMode=BMM_Additive
    rotationSpace=BCS_BoneSpace
    translationMode=BMM_Additive
    translationSpace=BCS_BoneSpace
    alphaBoolBlend {blendOption=Linear}
    boneToModify {boneName=dyn_back_l_01}
    rotation {yaw=-15}
   }
   n28 = "Transform (Modify) Bone - Bone: dyn_back_l_02"(ComponentPose=@n27, Alpha=@n26, Translation="0,0,0", Rotation="0,-25,0", Scale="1,1,1") {
    rotationMode=BMM_Additive
    rotationSpace=BCS_BoneSpace
    translationMode=BMM_Additive
    translationSpace=BCS_BoneSpace
    alphaBoolBlend {blendOption=Linear}
    boneToModify {boneName=dyn_back_l_02}
    rotation {yaw=-25}
   }
   n29 = "Transform (Modify) Bone - Bone: dyn_back_l_03"(ComponentPose=@n28, Alpha=@n26, Translation="0,0,0", Rotation="0,-10,0", Scale="1,1,1") {
    rotationMode=BMM_Additive
    rotationSpace=BCS_BoneSpace
    translationMode=BMM_Additive
    translationSpace=BCS_BoneSpace
    alphaBoolBlend {blendOption=Linear}
    boneToModify {boneName=dyn_back_l_03}
    rotation {yaw=-10}
   }
   n30 = "Transform (Modify) Bone - Bone: dyn_back_r_01"(ComponentPose=@n29, Alpha=@n26, Translation="0,0,0", Rotation="0,-15,0", Scale="1,1,1") {
    rotationMode=BMM_Additive
    rotationSpace=BCS_ParentBoneSpace
    translationMode=BMM_Additive
    translationSpace=BCS_BoneSpace
    alphaBoolBlend {blendOption=Linear}
    boneToModify {boneName=dyn_back_r_01}
    rotation {yaw=-15}
   }
   n31 = "Transform (Modify) Bone - Bone: dyn_back_r_02"(ComponentPose=@n30, Alpha=@n26, Translation="0,0,0", Rotation="0,-25,0", Scale="1,1,1") {
    rotationMode=BMM_Additive
    rotationSpace=BCS_ParentBoneSpace
    translationMode=BMM_Additive
    translationSpace=BCS_BoneSpace
    alphaBoolBlend {blendOption=Linear}
    boneToModify {boneName=dyn_back_r_02}
    rotation {yaw=-25}
   }
   n32 = "Transform (Modify) Bone - Bone: dyn_back_r_03"(ComponentPose=@n31, Alpha=@n26, Translation="0,0,0", Rotation="0,-10,0", Scale="1,1,1") {
    rotationMode=BMM_Additive
    rotationSpace=BCS_ParentBoneSpace
    translationMode=BMM_Additive
    translationSpace=BCS_BoneSpace
    alphaBoolBlend {blendOption=Linear}
    boneToModify {boneName=dyn_back_r_03}
    rotation {yaw=-10}
   }
   Save cached pose 'NO CROUCH SCALABILITY'(Pose=Component To Local(ComponentPose=@n32)) {cacheName=NO CROUCH SCALABILITY}
   n33 = "Transform (Modify) Bone - Bone: dyn_back_r_01"!(
    ComponentPose=Local To Component(LocalPose=Use cached pose 'SLIDING FIX'())
    bAlphaBoolEnabled:bool=True
    Translation="0,0,0"
    Rotation="0,0,0"
    Scale="1,1,1"
   ) -> (Pose) {
    alphaInputType=Bool
    rotationMode=BMM_Additive
    rotationSpace=BCS_ParentBoneSpace
    translationMode=BMM_Additive
    translationSpace=BCS_BoneSpace
    alphaBoolBlend {blendOption=Linear}
    boneToModify {boneName=dyn_back_r_01}
   }
   n34 = "Transform (Modify) Bone - Bone: dyn_back_r_02"!(ComponentPose=@n33, bAlphaBoolEnabled:bool=True, Translation="0,0,0", Rotation="0,0,0", Scale="1,1,1") -> (Pose) {
    alphaInputType=Bool
    rotationMode=BMM_Additive
    rotationSpace=BCS_ParentBoneSpace
    translationMode=BMM_Additive
    translationSpace=BCS_BoneSpace
    alphaBoolBlend {blendOption=Linear}
    boneToModify {boneName=dyn_back_r_02}
   }
   n35 = "Transform (Modify) Bone - Bone: dyn_back_r_03"!(ComponentPose=@n34, bAlphaBoolEnabled:bool=True, Translation="0,0,0", Rotation="0,0,0", Scale="1,1,1") -> (Pose) {
    alphaInputType=Bool
    rotationMode=BMM_Additive
    rotationSpace=BCS_ParentBoneSpace
    translationMode=BMM_Additive
    translationSpace=BCS_BoneSpace
    alphaBoolBlend {blendOption=Linear}
    boneToModify {boneName=dyn_back_r_03}
   }
   n36 = "Transform (Modify) Bone - Bone: dyn_back_l_01"!(ComponentPose=@n35, bAlphaBoolEnabled:bool=True, Translation="0,0,0", Rotation="0,0,0", Scale="1,1,1") -> (Pose) {
    alphaInputType=Bool
    rotationMode=BMM_Additive
    rotationSpace=BCS_BoneSpace
    translationMode=BMM_Additive
    translationSpace=BCS_BoneSpace
    alphaBoolBlend {blendOption=Linear}
    boneToModify {boneName=dyn_back_l_01}
   }
   n37 = "Transform (Modify) Bone - Bone: dyn_back_l_02"!(ComponentPose=@n36, bAlphaBoolEnabled:bool=True, Translation="0,0,0", Rotation="0,0,0", Scale="1,1,1") -> (Pose) {
    alphaInputType=Bool
    rotationMode=BMM_Additive
    rotationSpace=BCS_BoneSpace
    translationMode=BMM_Additive
    translationSpace=BCS_BoneSpace
    alphaBoolBlend {blendOption=Linear}
    boneToModify {boneName=dyn_back_l_02}
   }
   n38 = "Transform (Modify) Bone - Bone: dyn_back_l_03"!(ComponentPose=@n37, bAlphaBoolEnabled:bool=True, Translation="0,0,0", Rotation="0,0,0", Scale="1,1,1") -> (Pose) {
    alphaInputType=Bool
    rotationMode=BMM_Additive
    rotationSpace=BCS_BoneSpace
    translationMode=BMM_Additive
    translationSpace=BCS_BoneSpace
    alphaBoolBlend {blendOption=Linear}
    boneToModify {boneName=dyn_back_l_03}
   }
   n39 = Get Alpha_Thigh_L_Back()
   n40 = "Transform (Modify) Bone - Bone: dyn_back_r_01"!(ComponentPose=@n38, Alpha:float=@n39, Translation="0,0,0", Rotation="0,0,0", Scale="1,1,1") -> (Pose) {
    rotationMode=BMM_Additive
    rotationSpace=BCS_ParentBoneSpace
    translationMode=BMM_Additive
    translationSpace=BCS_BoneSpace
    alphaBoolBlend {blendOption=Linear}
    boneToModify {boneName=dyn_back_r_01}
   }
   n41 = "Transform (Modify) Bone - Bone: dyn_back_r_02"!(ComponentPose=@n40, Alpha:float=@n39, Translation="0,0,0", Rotation="0,0,0", Scale="1,1,1") -> (Pose) {
    rotationMode=BMM_Additive
    rotationSpace=BCS_ParentBoneSpace
    translationMode=BMM_Additive
    translationSpace=BCS_BoneSpace
    alphaBoolBlend {blendOption=Linear}
    boneToModify {boneName=dyn_back_r_02}
   }
   n42 = "Transform (Modify) Bone - Bone: dyn_back_r_03"!(ComponentPose=@n41, Alpha:float=@n39, Translation="0,0,0", Rotation="0,0,0", Scale="1,1,1") -> (Pose) {
    rotationMode=BMM_Additive
    rotationSpace=BCS_ParentBoneSpace
    translationMode=BMM_Additive
    translationSpace=BCS_BoneSpace
    alphaBoolBlend {blendOption=Linear}
    boneToModify {boneName=dyn_back_r_03}
   }
   n43 = "Transform (Modify) Bone - Bone: dyn_back_l_01"!(ComponentPose=@n42, Alpha:float=@n39, Translation="0,0,0", Rotation="0,0,0", Scale="1,1,1") -> (Pose) {
    rotationMode=BMM_Additive
    rotationSpace=BCS_BoneSpace
    translationMode=BMM_Additive
    translationSpace=BCS_BoneSpace
    alphaBoolBlend {blendOption=Linear}
    boneToModify {boneName=dyn_back_l_01}
   }
   n44 = "Transform (Modify) Bone - Bone: dyn_back_l_02"!(ComponentPose=@n43, Alpha:float=@n39, Translation="0,0,0", Rotation="0,0,0", Scale="1,1,1") -> (Pose) {
    rotationMode=BMM_Additive
    rotationSpace=BCS_BoneSpace
    translationMode=BMM_Additive
    translationSpace=BCS_BoneSpace
    alphaBoolBlend {blendOption=Linear}
    boneToModify {boneName=dyn_back_l_02}
   }
   n45 = "Transform (Modify) Bone - Bone: dyn_back_l_03"!(ComponentPose=@n44, Alpha:float=@n39, Translation="0,0,0", Rotation="0,0,0", Scale="1,1,1") -> (Pose) {
    rotationMode=BMM_Additive
    rotationSpace=BCS_BoneSpace
    translationMode=BMM_Additive
    translationSpace=BCS_BoneSpace
    alphaBoolBlend {blendOption=Linear}
    boneToModify {boneName=dyn_back_l_03}
   }
   n46 = Get Alpha_Thigh_R_Back()
   n47 = "Transform (Modify) Bone - Bone: dyn_back_l_01"(ComponentPose=@n45, Alpha=@n46, Translation="0,0,0", Rotation="0,0,0", Scale="1,1,1") {
    rotationMode=BMM_Additive
    rotationSpace=BCS_BoneSpace
    translationMode=BMM_Additive
    translationSpace=BCS_BoneSpace
    alphaBoolBlend {blendOption=Linear}
    boneToModify {boneName=dyn_back_l_01}
   }
   n48 = "Transform (Modify) Bone - Bone: dyn_back_l_02"(ComponentPose=@n47, Alpha=@n46, Translation="0,0,0", Rotation="0,0,0", Scale="1,1,1") {
    rotationMode=BMM_Additive
    rotationSpace=BCS_BoneSpace
    translationMode=BMM_Additive
    translationSpace=BCS_BoneSpace
    alphaBoolBlend {blendOption=Linear}
    boneToModify {boneName=dyn_back_l_02}
   }
   n49 = "Transform (Modify) Bone - Bone: dyn_back_l_03"(ComponentPose=@n48, Alpha=@n46, Translation="0,0,0", Rotation="0,0,0", Scale="1,1,1") {
    rotationMode=BMM_Additive
    rotationSpace=BCS_BoneSpace
    translationMode=BMM_Additive
    translationSpace=BCS_BoneSpace
    alphaBoolBlend {blendOption=Linear}
    boneToModify {boneName=dyn_back_l_03}
   }
   n50 = "Transform (Modify) Bone - Bone: dyn_back_r_01"(ComponentPose=@n49, Alpha=@n46, Translation="0,0,0", Rotation="0,0,0", Scale="1,1,1") {
    rotationMode=BMM_Additive
    rotationSpace=BCS_ParentBoneSpace
    translationMode=BMM_Additive
    translationSpace=BCS_BoneSpace
    alphaBoolBlend {blendOption=Linear}
    boneToModify {boneName=dyn_back_r_01}
   }
   n51 = "Transform (Modify) Bone - Bone: dyn_back_r_02"(ComponentPose=@n50, Alpha=@n46, Translation="0,0,0", Rotation="0,0,0", Scale="1,1,1") {
    rotationMode=BMM_Additive
    rotationSpace=BCS_ParentBoneSpace
    translationMode=BMM_Additive
    translationSpace=BCS_BoneSpace
    alphaBoolBlend {blendOption=Linear}
    boneToModify {boneName=dyn_back_r_02}
   }
   n52 = "Transform (Modify) Bone - Bone: dyn_back_r_03"(ComponentPose=@n51, Alpha=@n46, Translation="0,0,0", Rotation="0,0,0", Scale="1,1,1") {
    rotationMode=BMM_Additive
    rotationSpace=BCS_ParentBoneSpace
    translationMode=BMM_Additive
    translationSpace=BCS_BoneSpace
    alphaBoolBlend {blendOption=Linear}
    boneToModify {boneName=dyn_back_r_03}
   }
   Save cached pose 'CROUCH SCALABILITY'(Pose=Component To Local(ComponentPose=@n52)) {cacheName=CROUCH SCALABILITY}
   n53 = "Transform (Modify) Bone - Bone: dyn_back_r_01"!(
    ComponentPose=Local To Component(LocalPose=Use cached pose 'SLIDING FIX'())
    bAlphaBoolEnabled:bool=True
    Translation="0,0,0"
    Rotation="0,15,0"
    Scale="1,1,1"
   ) -> (Pose) {
    alphaInputType=Bool
    rotationMode=BMM_Additive
    rotationSpace=BCS_ParentBoneSpace
    translationMode=BMM_Additive
    translationSpace=BCS_BoneSpace
    alphaBoolBlend {blendOption=Linear}
    boneToModify {boneName=dyn_back_r_01}
    rotation {yaw=15}
   }
   n54 = "Transform (Modify) Bone - Bone: dyn_back_r_02"!(ComponentPose=@n53, bAlphaBoolEnabled:bool=True, Translation="0,0,0", Rotation="0,15,0", Scale="1,1,1") -> (Pose) {
    alphaInputType=Bool
    rotationMode=BMM_Additive
    rotationSpace=BCS_ParentBoneSpace
    translationMode=BMM_Additive
    translationSpace=BCS_BoneSpace
    alphaBoolBlend {blendOption=Linear}
    boneToModify {boneName=dyn_back_r_02}
    rotation {yaw=15}
   }
   n55 = "Transform (Modify) Bone - Bone: dyn_back_r_03"!(ComponentPose=@n54, bAlphaBoolEnabled:bool=True, Translation="0,0,0", Rotation="0,0,0", Scale="1,1,1") -> (Pose) {
    alphaInputType=Bool
    rotationMode=BMM_Additive
    rotationSpace=BCS_ParentBoneSpace
    translationMode=BMM_Additive
    translationSpace=BCS_BoneSpace
    alphaBoolBlend {blendOption=Linear}
    boneToModify {boneName=dyn_back_r_03}
   }
   n56 = "Transform (Modify) Bone - Bone: dyn_back_l_01"!(ComponentPose=@n55, bAlphaBoolEnabled:bool=True, Translation="0,0,0", Rotation="0,15,0", Scale="1,1,1") -> (Pose) {
    alphaInputType=Bool
    rotationMode=BMM_Additive
    rotationSpace=BCS_BoneSpace
    translationMode=BMM_Additive
    translationSpace=BCS_BoneSpace
    alphaBoolBlend {blendOption=Linear}
    boneToModify {boneName=dyn_back_l_01}
    rotation {yaw=15}
   }
   n57 = "Transform (Modify) Bone - Bone: dyn_back_l_02"!(ComponentPose=@n56, bAlphaBoolEnabled:bool=True, Translation="0,0,0", Rotation="0,15,0", Scale="1,1,1") -> (Pose) {
    alphaInputType=Bool
    rotationMode=BMM_Additive
    rotationSpace=BCS_BoneSpace
    translationMode=BMM_Additive
    translationSpace=BCS_BoneSpace
    alphaBoolBlend {blendOption=Linear}
    boneToModify {boneName=dyn_back_l_02}
    rotation {yaw=15}
   }
   n58 = "Transform (Modify) Bone - Bone: dyn_back_l_03"!(ComponentPose=@n57, bAlphaBoolEnabled:bool=True, Translation="0,0,0", Rotation="0,0,0", Scale="1,1,1") -> (Pose) {
    alphaInputType=Bool
    rotationMode=BMM_Additive
    rotationSpace=BCS_BoneSpace
    translationMode=BMM_Additive
    translationSpace=BCS_BoneSpace
    alphaBoolBlend {blendOption=Linear}
    boneToModify {boneName=dyn_back_l_03}
   }
   Save cached pose 'DNBO SCALABILITY'(Pose=Component To Local(ComponentPose=@n58)) {cacheName=DNBO SCALABILITY}
   n59 = "Transform (Modify) Bone - Bone: dyn_back_r_02"!(
    ComponentPose=Local To Component(LocalPose=Use cached pose 'SCALABILITY input'())
    Alpha:float=1
    Translation="0,0,0"
    Rotation="0,-20,0"
    Scale="1,1,1"
   ) -> (Pose) {
    rotationMode=BMM_Additive
    rotationSpace=BCS_ParentBoneSpace
    translationMode=BMM_Additive
    translationSpace=BCS_BoneSpace
    alphaBoolBlend {blendOption=Linear}
    boneToModify {boneName=dyn_back_r_02}
    rotation {yaw=-20}
   }
   n60 = "Transform (Modify) Bone - Bone: dyn_back_r_03"!(ComponentPose=@n59, Alpha:float=1, Translation="0,0,0", Rotation="0,-25,0", Scale="1,1,1") -> (Pose) {
    rotationMode=BMM_Additive
    rotationSpace=BCS_ParentBoneSpace
    translationMode=BMM_Additive
    translationSpace=BCS_BoneSpace
    alphaBoolBlend {blendOption=Linear}
    boneToModify {boneName=dyn_back_r_03}
    rotation {yaw=-25}
   }
   n61 = "Transform (Modify) Bone - Bone: dyn_back_l_02"!(ComponentPose=@n60, Alpha:float=1, Translation="0,0,0", Rotation="0,-20,0", Scale="1,1,1") -> (Pose) {
    rotationMode=BMM_Additive
    rotationSpace=BCS_BoneSpace
    translationMode=BMM_Additive
    translationSpace=BCS_BoneSpace
    alphaBoolBlend {blendOption=Linear}
    boneToModify {boneName=dyn_back_l_02}
    rotation {yaw=-20}
   }
   n62 = "Transform (Modify) Bone - Bone: dyn_back_l_03"!(ComponentPose=@n61, Alpha:float=1, Translation="0,0,0", Rotation="0,-25,0", Scale="1,1,1") -> (Pose) {
    rotationMode=BMM_Additive
    rotationSpace=BCS_BoneSpace
    translationMode=BMM_Additive
    translationSpace=BCS_BoneSpace
    alphaBoolBlend {blendOption=Linear}
    boneToModify {boneName=dyn_back_l_03}
    rotation {yaw=-25}
   }
   Save cached pose 'SKYDIVING SCALABILITY'(Pose=Component To Local(ComponentPose=@n62)) {cacheName=SKYDIVING SCALABILITY}
   n63 = "Transform (Modify) Bone - Bone: dyn_back_r_01"!(
    ComponentPose=Local To Component(LocalPose=Use cached pose 'SLIDING FIX'())
    Alpha:float=1
    Translation="0,-1,0"
    Rotation="0,0,0"
    Scale="1,1,1"
   ) -> (Pose) {
    rotationMode=BMM_Additive
    rotationSpace=BCS_ParentBoneSpace
    translationMode=BMM_Additive
    translationSpace=BCS_BoneSpace
    alphaBoolBlend {blendOption=Linear}
    boneToModify {boneName=dyn_back_r_01}
    translation {y=-1}
   }
   n64 = "Transform (Modify) Bone - Bone: dyn_back_l_01"!(ComponentPose=@n63, Alpha:float=1, Translation="0,-1,0", Rotation="0,0,0", Scale="1,1,1") -> (Pose) {
    rotationMode=BMM_Additive
    rotationSpace=BCS_ParentBoneSpace
    translationMode=BMM_Additive
    translationSpace=BCS_BoneSpace
    alphaBoolBlend {blendOption=Linear}
    boneToModify {boneName=dyn_back_l_01}
    translation {y=-1}
   }
   n65 = "Transform (Modify) Bone - Bone: dyn_back_r_01"(ComponentPose=@n64, Alpha=1, Translation="0,0,0", Rotation="0,-5,0", Scale="1,1,1") {
    rotationMode=BMM_Additive
    rotationSpace=BCS_ParentBoneSpace
    translationMode=BMM_Additive
    translationSpace=BCS_BoneSpace
    alphaBoolBlend {blendOption=Linear}
    boneToModify {boneName=dyn_back_r_01}
    rotation {yaw=-5}
   }
   n66 = "Transform (Modify) Bone - Bone: dyn_back_l_01"(ComponentPose=@n65, Alpha=1, Translation="0,0,0", Rotation="0,-5,0", Scale="1,1,1") {
    rotationMode=BMM_Additive
    rotationSpace=BCS_BoneSpace
    translationMode=BMM_Additive
    translationSpace=BCS_BoneSpace
    alphaBoolBlend {blendOption=Linear}
    boneToModify {boneName=dyn_back_l_01}
    rotation {yaw=-5}
   }
   Save cached pose 'SCALABILITY input'(Pose=Component To Local(ComponentPose=@n66)) {cacheName=SCALABILITY input}
   n67 = RigidBody!(
    ComponentPose=Local To Component(LocalPose=Use cached pose 'BackRBANCache'())
    Alpha=Get RBANSprintAlpha() -> (RBANSprintAlpha:double)
    OverrideWorldGravity=Get Gravity Override()
    ExternalForce
   ) -> (Pose) {
    bClampLinearTranslationLimitToRefPose=true
    bOverrideWorldGravity=true
    overridePhysicsAsset=&/F_MED_RemoteControl_Physics_FrontSkirt_CrouchNew
    simulationSpace=BaseBoneSpace
    alphaBoolBlend {blendOption=Linear}
    baseBoneRef {boneName=spine_01}
    componentAppliedLinearAccClamp=<6000 6000 6000>
    componentLinearAccScale=<0.1 0.6 0.6>
    componentLinearVelScale=<0.1 0.6 0.6>
   }
   n68 = RigidBody(
    ComponentPose=Local To Component(LocalPose=Use cached pose 'BackRBANCache'())
    Alpha=Get RBANSprintAlpha()
    OverrideWorldGravity=Get Gravity Override()
   ) {
    bClampLinearTranslationLimitToRefPose=true
    bOverrideWorldGravity=true
    overridePhysicsAsset=&/F_MED_RemoteControl_Physics_FrontSkirt_FrontEnd
    simulationSpace=BaseBoneSpace
    alphaBoolBlend {blendOption=Linear}
    baseBoneRef {boneName=pelvis}
    componentAppliedLinearAccClamp=<6000 6000 6000>
    componentLinearAccScale=<0.1 0.6 0.6>
    componentLinearVelScale=<0.1 0.6 0.6>
   }
   n69 = "Transform (Modify) Bone - Bone: MSTR_front_l_01"(ComponentPose=@n68, Alpha=Get Alpha_Thigh_L_Front() -> (Alpha_Thigh_L_Front:float), Translation="0,0,0", Rotation="0,-10,0", Scale="1,1,1") {
    rotationMode=BMM_Additive
    rotationSpace=BCS_BoneSpace
    translationMode=BMM_Additive
    translationSpace=BCS_BoneSpace
    alphaBoolBlend {blendOption=Linear}
    boneToModify {boneName=MSTR_front_l_01}
    rotation {yaw=-10}
   }
   n70 = "Transform (Modify) Bone - Bone: MSTR_front_r_01"(ComponentPose=@n69, Alpha=Get Alpha_Thigh_R_Front() -> (Alpha_Thigh_R_Front:float), Translation="0,0,0", Rotation="0,-10,0", Scale="1,1,1") {
    rotationMode=BMM_Additive
    rotationSpace=BCS_BoneSpace
    translationMode=BMM_Additive
    translationSpace=BCS_BoneSpace
    alphaBoolBlend {blendOption=Linear}
    boneToModify {boneName=MSTR_front_r_01}
    rotation {yaw=-10}
   }
   Save cached pose 'POST'(Pose=Blend Poses by bool(BlendPose_0=Component To Local(ComponentPose=@n70), BlendPose_1=Component To Local(ComponentPose=@n67), bActiveValue=Get bIsInFrontEnd()) {blendType=Linear}) {cacheName=POST}
   n71 = Anim Dynamics - Bone: dyn_l_pants(
    ComponentPose=Local To Component(LocalPose=Use cached pose 'CONTROL RIG'())
    Alpha=Get pants alpha dynamics() -> (pants alpha dynamics:double)
    ExternalForce="0,0,0"
   ) {
    angularBiasOverride=0.8
    angularDampingOverride=0.9
    angularSpringConstant=250
    bLinearSpring=true
    bOverrideLinearDamping=true
    bUseGravityOverride=true
    linearDampingOverride=1
    linearSpringConstant=80
    alphaBoolBlend {blendOption=Linear}
    boundBone {boneName=dyn_l_pants}
    componentAppliedLinearAccClamp=<3000 3000 3000>
    componentLinearAccScale=<0.1 0.1 0.1>
    componentLinearVelScale=<0.9 0.9 0.9>
    physicsBodyDefinitions [{
      boundBone {boneName=dyn_l_pants}
      boxExtents=<20 20 20>
      constraintSetup {angularLimitsMax=<0 0 0>, angularLimitsMin=<0 0 0>, angularTarget=<0 0 0>, linearAxesMax=<30 30 30>, linearAxesMin=<-30 -30 -30>}
      localJointOffset=<0 0 0>
     }]
   }
   n72 = Anim Dynamics - Bone: dyn_r_pants(ComponentPose=@n71, Alpha=Get pants alpha dynamics(), ExternalForce="0,0,0") {
    angularBiasOverride=0.8
    angularDampingOverride=0.9
    angularSpringConstant=250
    bLinearSpring=true
    bOverrideLinearDamping=true
    bUseGravityOverride=true
    linearDampingOverride=1
    linearSpringConstant=80
    alphaBoolBlend {blendOption=Linear}
    boundBone {boneName=dyn_r_pants}
    componentAppliedLinearAccClamp=<3000 3000 3000>
    componentLinearAccScale=<0.1 0.1 0.1>
    componentLinearVelScale=<0.9 0.9 0.9>
    physicsBodyDefinitions [{
      boundBone {boneName=dyn_r_pants}
      boxExtents=<20 20 20>
      constraintSetup {angularLimitsMax=<0 0 0>, angularLimitsMin=<0 0 0>, angularTarget=<0 0 0>, linearAxesMax=<30 30 30>, linearAxesMin=<-30 -30 -30>}
      localJointOffset=<0 0 0>
     }]
   }
   Save cached pose 'LOWER PANTS'(Pose=Component To Local(ComponentPose=@n72)) {cacheName=LOWER PANTS}
   n73 = Control Rig!(
    Source=Copy Pose From Mesh() {bUseAttachedParent=true}
    Alpha
    isCrouching:bool=Get bIsCrouching()
    isSprintTact:bool=Get isSprintTact() -> (isSprintTact:bool)
   ) -> (Pose) {
    alphaBoolBlend {blendOption=Linear}
    controlRigAssetReference {blueprintRigClass=&/F_MED_RemoteControl_CtrlRig.F_MED_RemoteControl_CtrlRig_C}
   }
   Save cached pose 'CONTROL RIG'(Pose=@n73) {cacheName=CONTROL RIG}
   n74 = RigidBody!(
    ComponentPose=Local To Component(LocalPose=Use cached pose 'Pose'())
    Alpha
    OverrideWorldGravity="0,0,-980"
    ExternalForce
    ComponentLinearAccScale:Vector="0.6,0.6,0.6"
    ComponentLinearVelScale:Vector="0.6,0.6,0.6"
   ) -> (Pose) {
    bClampLinearTranslationLimitToRefPose=true
    bOverrideWorldGravity=true
    overridePhysicsAsset=&F_MED_RemoteControl_Physics_BackSkirt
    simulationSpace=BaseBoneSpace
    alphaBoolBlend {blendOption=Linear}
    baseBoneRef {boneName=pelvis}
    componentAppliedLinearAccClamp=<7000 7000 7000>
    componentLinearAccScale=<0.4 0.4 0.4>
    componentLinearVelScale=<0.6 0.6 0.6>
    overrideWorldGravity {z=-980}
   }
   n75 = RigidBody(
    ComponentPose=@n74
    Alpha=0.5
    OverrideWorldGravity="0,0,-980"
    ExternalForce=Break Fort Anim Input Skydiving External Force(FortAnimInput_SkydivingExternalForce=Get SkydivingExternalForceInput()) -> (
     bUseSkydivingVectorForce=~
     SkydivingVectorForce=~
     HeadToPelvisDirection=~
     FloatingMultiplier=~
     FloatingAdditive=~
     DivingMultiplier=~
     DivingAdditive=~
     ParachutingMultiplier=~
     ParachutingAdditive=~
    ).SkydivingVectorForce
    ComponentLinearAccScale="0.1,0.1,0.1"
    ComponentLinearVelScale="0.1,0.1,0.1"
   ) {
    bClampLinearTranslationLimitToRefPose=true
    bOverrideWorldGravity=true
    cachedBoundsScale=1
    overridePhysicsAsset=&F_MED_RemoteControl_Physics_FrontSkirt_CrouchNew
    simulationSpace=BaseBoneSpace
    alphaBoolBlend {blendOption=Linear}
    baseBoneRef {boneName=pelvis}
    componentAppliedLinearAccClamp=<5000 5000 5000>
    componentLinearAccScale=<0.4 0.4 0.4>
    componentLinearVelScale=<0.6 0.6 0.6>
    overrideWorldGravity {z=-980}
   }
   Save cached pose 'RBANoutput'(Pose=Blend Poses by bool(
     BlendPose_0=Component To Local(ComponentPose=@n75)
     BlendPose_1=Use cached pose 'POST'()
     BlendTime_0=0.3
     bActiveValue=Get bIsTacticalSprinting()
    ) {blendType=Linear}) {cacheName=RBANoutput}
   n76 = Blend Poses by bool(
    BlendPose_0=Use cached pose 'CROUCH SCALABILITY'()
    BlendPose_1=Use cached pose 'NO CROUCH SCALABILITY'()
    BlendTime_0=0.3
    bActiveValue=Get bIsCrouching()
   ) {blendType=Linear}
   n77 = Use cached pose 'SKYDIVING SCALABILITY'()
   n78 = Get bIsSkyDiving()
   Save cached pose 'SCALABILITY output'(Pose=Blend Poses by bool(
     BlendPose_0=Use cached pose 'DNBO SCALABILITY'()
     BlendPose_1=Blend Poses by bool(BlendPose_0=@n77, BlendPose_1=@n76, BlendTime_0=0.3, bActiveValue=@n78) {blendType=Linear}
     BlendTime_0=0.3
     bActiveValue=Get bIsDBNO()
    ) {blendType=Linear}) {cacheName=SCALABILITY output}
   n79 = "Constraint - dyn_back_r_01 (PPP)"!(
    ComponentPose=Local To Component(LocalPose=Use cached pose 'Pose'())
    Alpha=1
    ConstraintWeights_0=0
    ConstraintWeights_1:float=1
    ConstraintWeights_2:float=0
   ) -> (Pose) {
    alphaBoolBlend {blendOption=Linear}
    boneToModify {boneName=dyn_back_r_01}
    constraintSetup [{targetBone {boneName=thigh_r}}, {targetBone {boneName=pelvis}}, {targetBone {boneName=thigh_twist_01_r}}]
    constraintWeights [1,1,1]
   }
   n80 = "Constraint - dyn_back_l_01 (PPP)"!(ComponentPose=@n79, Alpha=1, ConstraintWeights_0=0, ConstraintWeights_1:float=1, ConstraintWeights_2:float=0) -> (Pose) {
    alphaBoolBlend {blendOption=Linear}
    boneToModify {boneName=dyn_back_l_01}
    constraintSetup [{targetBone {boneName=thigh_l}}, {targetBone {boneName=pelvis}}, {targetBone {boneName=thigh_twist_01_l}}]
    constraintWeights [1,1,1]
   }
   Save cached pose 'ALT CONSTRAINT'(Pose=Component To Local(ComponentPose=@n80)) {cacheName=ALT CONSTRAINT}
   Break Fort Anim Input Skydiving External Force(FortAnimInput_SkydivingExternalForce=Get SkydivingExternalForceInput()) -> (
    bUseSkydivingVectorForce=~
    SkydivingVectorForce=~
    HeadToPelvisDirection=~
    FloatingMultiplier=~
    FloatingAdditive=~
    DivingMultiplier=~
    DivingAdditive=~
    ParachutingMultiplier=~
    ParachutingAdditive=~
   )
   Get 30 FPS correction() -> (30 FPS correction:double)
  ]}
 EventGraph {nodes [
   Event OnCharacterCustomizationLoaded()
   > n1 = Set SkelMesh(SkelMesh=GetPartSkeletalMeshComponent())
   > n2 = Get Dynamic Material Instance(SkeletalMeshComponent=@n1)
   > Set Mat Instance Dynamic(Mat Instance Dynamic=@n2)
   > n3 = GetOrCreateDynamicOverlayMaterial(MeshComponent=@n1, MaterialName=None)
   > Set Overlay Mat Instance Dynamic(Overlay Mat Instance Dynamic=@n3)
   n4 = CheckForBackbling()
   > Set WearingBack(WearingBack:bool=@n4) -> (Output_Get:bool)
   > Set WearingCape(WearingCape:bool=AND Boolean(A=@n4, B=@n4.bIsBackblingACape)) -> (Output_Get:bool)
   Event BlueprintInitializeAnimation()
   > Set SkelMesh(SkelMesh=GetOwnerSkeletalMeshComponent())
   n5 = TryGetPawnOwner()
   > Is Valid(InputObject=@n5)
   > n6 = Cast To FortPlayerPawnAthena(Object=@n5)
   > Set PlayerPawn(PlayerPawn:Fort Player Pawn Athena Object Reference=@n6) -> (Output_Get:Fort Player Pawn Athena Object Reference)
   n7 = Event BlueprintUpdateAnimation()
   > n8 = Set Delta Time X(Delta Time X=@n7.DeltaTimeX)
   > Set 30 FPS correction(30 FPS correction:double=MapRangeClamped(Value=float / float(A=1, B=@n8), InRangeA=25, InRangeB=60)) -> (Output_Get:double)
   > n9 = Sequence!(execute) -> (then_0, then_1, then_2:Exec, then_3:Exec, then_4:Exec, then_5:Exec, then_6:Exec, then_7:Exec, then_8:Exec, then_9:Exec, then_10:Exec)
   n10 = Get PawnSpeed2D()
   n11 = Branch(execute=@n9.then_8, Condition=AND Boolean(A=Get bIsMoving2D(), B="float >= float"(A=@n10, B=150)))
   > Set Is moving alpha(Is moving alpha:float=1) -> (Output_Get:float)
   Set Is moving alpha(execute=@n11.else, Is moving alpha=0)
   n12 = Branch(execute=@n9.then_9, Condition=OR Boolean(A=Get bIsSprinting(), B=Get bIsTacticalSprinting()))
   > Set isSprintTact(isSprintTact:bool=true) -> (Output_Get:bool)
   > Set RBANSprintAlpha(RBANSprintAlpha:double=0.5) -> (Output_Get:double)
   Set isSprintTact(execute=@n12.else, isSprintTact=false)
   > Set RBANSprintAlpha(RBANSprintAlpha=1)
   n13 = NOT Boolean(A=Get Is moving forward() -> (Is moving forward:bool))
   n14 = NOT Boolean(A=Get bIsFalling())
   n15 = NOT Boolean(A=Get bIsJumping())
   n16 = float * float(
    A="To Float (Boolean)"(InBool=AND Boolean!(A=@n14, B=@n15, C:bool=@n13) -> (ReturnValue))
    B=GetAlphaFromReferenceTransformDelta(SkeletalMeshComponent=Get SkelMesh(), BoneName=thigh_l, BaseName=pelvis, InputRange=(X=0,Y=-40), TwistAxis="0,0,1")
   ) {is=title K2Node_PromotableOperator}
   Set Alpha_Thigh_L_Front(execute=@n9.then_7, Alpha_Thigh_L_Front:float=@n16) -> (Output_Get:float)
   n17 = NOT Boolean(A=Get Is moving forward())
   n18 = NOT Boolean(A=Get bIsFalling())
   n19 = NOT Boolean(A=Get bIsJumping())
   n20 = float * float(
    A="To Float (Boolean)"(InBool=AND Boolean(A=@n18, B=@n19, C=@n17))
    B=GetAlphaFromReferenceTransformDelta(SkeletalMeshComponent=Get SkelMesh(), BoneName=thigh_r, BaseName=pelvis, InputRange=(X=0,Y=-40), TwistAxis="0,0,1")
   )
   > Set Alpha_Thigh_R_Front(Alpha_Thigh_R_Front:float=@n20) -> (Output_Get:float)
   n21 = NOT Boolean(A=Get bIsJumping())
   n22 = NOT Boolean(A=Get bIsFalling())
   n23 = "float > float"(A=Get PawnSpeed2D(), B=225)
   n24 = float * float(
    A="To Float (Boolean)"(InBool=AND Boolean(A=@n22, B=@n21, C=@n23))
    B=GetAlphaFromReferenceTransformDelta(SkeletalMeshComponent=Get SkelMesh(), BoneName=thigh_l, BaseName=pelvis, InputRange=(X=-20,Y=-70), TwistAxis="0,0,1")
   )
   > Set Alpha_Thigh_L_Back(Alpha_Thigh_L_Back:double=@n24) -> (Output_Get:double)
   n25 = NOT Boolean(A=Get bIsFalling())
   n26 = NOT Boolean(A=Get bIsJumping())
   n27 = "float > float"(A=Get PawnSpeed2D(), B=225)
   n28 = float * float(
    A="To Float (Boolean)"(InBool=AND Boolean(A=@n25, B=@n26, C=@n27))
    B=GetAlphaFromReferenceTransformDelta(SkeletalMeshComponent=Get SkelMesh(), BoneName=thigh_r, BaseName=pelvis, InputRange=(X=-20,Y=-70), TwistAxis="0,0,1")
   )
   > Set Alpha_Thigh_R_Back(Alpha_Thigh_R_Back:double=@n28) -> (Output_Get:double)
   n29 = GetAlphaFromReferenceTransformDelta(SkeletalMeshComponent=Get SkelMesh(), BoneName=thigh_l, BaseName=pelvis, InputRange=(X=-70,Y=-50), TwistAxis="0,0,1")
   n30 = NOT Boolean(A=Get bIsFalling())
   n31 = NOT Boolean(A=Get bIsJumping())
   n32 = "float > float"(A=Get PawnSpeed2D(), B=225)
   n33 = float * float(A="To Float (Boolean)"(InBool=AND Boolean(A=@n30, B=@n31, C=@n32)), B=@n29)
   > Set Alpha_Thigh_LBackCross(Alpha_Thigh_LBackCross:double=@n33) -> (Output_Get:double)
   > PrintString!(execute, InString="To String (Float)"(InDouble=@n29.RawOutput), bPrintToScreen, bPrintToLog, TextColor, Duration=0, Key=1) -> (then)
   n34 = AND Boolean!(
    A=NOT Boolean(A=Get bIsSkyDiving())
    B=NOT Boolean(A=Get bIsSkydiveDiveMode())
    C=NOT Boolean(A=Get bIsSkydiveFloating())
    D:bool=NOT Boolean(A=Get bIsParachuteOpened())
   ) -> (ReturnValue)
   n35 = Get SkelMesh()
   > Set upperarm_l alpha(upperarm_l alpha:float=float * float(
     A="To Float (Boolean)"(InBool=@n34)
     B=GetAlphaFromReferenceTransformDelta(SkeletalMeshComponent=@n35, BoneName=upperarm_l, BaseName=spine_05, DeltaComponentType=EulerY, InputRange=(X=-10,Y=-25), TwistAxis="0,0,1")
    )) -> (Output_Get:float)
   n36 = AND Boolean(
    A=NOT Boolean(A=Get bIsSkyDiving())
    B=NOT Boolean(A=Get bIsSkydiveDiveMode())
    C=NOT Boolean(A=Get bIsSkydiveFloating())
    D=NOT Boolean(A=Get bIsParachuteOpened())
   )
   n37 = Get SkelMesh()
   > Set upperarm_r alpha(upperarm_r alpha:float=float * float(
     A="To Float (Boolean)"(InBool=@n36)
     B=GetAlphaFromReferenceTransformDelta(SkeletalMeshComponent=@n37, BoneName=upperarm_r, BaseName=clavicle_r, DeltaComponentType=EulerY, InputRange=(X=-10,Y=-25), TwistAxis="0,0,1")
    )) -> (Output_Get:float)
   > Set Is crouching and moving(Is crouching and moving:bool=AND Boolean!(A=Get bIsMoving2D(), B=Get bIsCrouching()) -> (ReturnValue)) -> (Output_Get:bool)
   n38 = Get PawnSpeed2D()
   > n39 = Branch(Condition=AND Boolean!(A=Get bIsMoving2D(), B=Get bIsCrouching(), C:bool="float >= float"(A=@n38, B=50)) -> (ReturnValue))
   > Set Is Crouching and moving alpha(Is Crouching and moving alpha:double=1) -> (Output_Get:double)
   Set Is Crouching and moving alpha(execute=@n39.else, Is Crouching and moving alpha=0)
   n40 = Get SkelMesh()
   n41 = Get bIsSkydiveFloating()
   n42 = Get bIsSkyDiving()
   n43 = AND Boolean!(A=OR Boolean(A=Get bIsJumping(), B=Get bIsFalling()), B=NOT Boolean(A=OR Boolean(A=@n41, B=@n42))) -> (ReturnValue)
   n44 = "Set  jump alpha"(
    execute=@n9.then_6
    " jump alpha":double=float * float(
     A="To Float (Boolean)"(InBool=@n43)
     B=GetAlphaFromReferenceTransformDelta(SkeletalMeshComponent=@n40, BoneName=thigh_r, BaseName=pelvis, InputRange=(X=0,Y=-70), TwistAxis="0,0,1")
    )
   ) -> (Output_Get:double)
   n45 = Get Is moving forward()
   > Set jump alpha Forward(jump alpha Forward:double=float * float(A="To Float (Boolean)"(InBool=@n45), B=@n44)) -> (Output_Get:double)
   > Set calf_r_alpha(calf_r_alpha:float=GetAlphaFromReferenceTransformDelta(SkeletalMeshComponent=Get SkelMesh(), BoneName=calf_r, BaseName=thigh_r, InputRange=(X=20,Y=115), TwistAxis="0,0,1")) -> (Output_Get:float)
   > Set calf_l_alpha(calf_l_alpha:float=GetAlphaFromReferenceTransformDelta(SkeletalMeshComponent=Get SkelMesh(), BoneName=calf_l, BaseName=thigh_l, InputRange=(X=20,Y=115), TwistAxis="0,0,1")) -> (Output_Get:float)
   > n46 = Branch(Condition=OR Boolean!(A=Get bIsInFrontEnd(), B=Get bIsMoving2D(), C:bool=Get bIsPlayingEmote()) -> (ReturnValue))
   > Set pants alpha dynamics(pants alpha dynamics:double=0.3) -> (Output_Get:double)
   Set pants alpha dynamics(execute=@n46.else, pants alpha dynamics=1)
   n47 = Get Is moving forward()
   n48 = Get PawnSpeed2D()
   Set bachcloth fwd(execute=@n9.then_4, bachcloth fwd:double=SelectFloat(A="To Float (Boolean)"(InBool=@n47), B=0, bPickA="float >= float"(A=@n48, B=500))) -> (Output_Get:double)
   Is Valid(exec=@n9.then_0, InputObject=Get PlayerPawn() -> (PlayerPawn:Fort Player Pawn Athena Object Reference))
   > Set WPO Globals Function(
    Material=Get Mat Instance Dynamic()
    Overlay Material=Get Overlay Mat Instance Dynamic()
    Skeletal Mesh=Get SkelMesh()
    "Local Noise Transform  Location"="Get Local Noise Transform  Location"() -> ("Local Noise Transform  Location":Vector)
    Local Noise Offset=Get Local Noise Offset()
   )
   n49 = Get Delta Time X()
   > Set WPO Regional Function!(
    execute
    Material:Material Instance Dynamic Object Reference=Get Mat Instance Dynamic()
    Overlay Material:Material Instance Dynamic Object Reference=Get Overlay Mat Instance Dynamic()
    Skeletal Mesh:Skeletal Mesh Component Object Reference=Get SkelMesh()
    Velocity Transform:Name=foot_r
    Velocity Interp Speed:double=1
    Collision Transform:Name=calf_r
    Collision Transform Axis:EAxis Enum
    Invert Collision Transform Axis:bool=false
    Delta Time X:double=@n49
    Previous Velocity Transform Location:"Vector (by ref)"=Get R1 Velocity Transform Location()
    Previous Velocity Transform Velocity:"Vector (by ref)"=Get R1 Velocity Transform Velocity()
    Region:int=1
   ) -> (then)
   > Set WPO Regional Function(
    Material=Get Mat Instance Dynamic()
    Overlay Material=Get Overlay Mat Instance Dynamic()
    Skeletal Mesh=Get SkelMesh()
    Velocity Transform=foot_l
    Velocity Interp Speed=1
    Collision Transform=calf_l
    Invert Collision Transform Axis=true
    Delta Time X=@n49
    Previous Velocity Transform Location=Get R2 Velocity Transform Location()
    Previous Velocity Transform Velocity=Get R2 Velocity Transform Velocity()
    Region=2
   )
   > n50 = Branch(Condition=Get bIsSkydiveFloating())
   > n51 = Set WPO Pho Wind Intensity(WPO Pho Wind Intensity=1)
   SetScalarParameterValue(execute=[@n52.then, @n51.then], self=Get Mat Instance Dynamic(), ParameterName=ClothSim_WindStrength, Value=Get WPO Pho Wind Intensity())
   n52 = Set WPO Pho Wind Intensity(execute=@n50.else, WPO Pho Wind Intensity=0)
   n53 = Branch(execute=@n9.then_3, Condition=Get bIsJumping())
   > Set AccScaleVals(AccScaleVals:Vector="0.1,0.1,0") -> (Output_Get:Vector)
   > n54 = Branch(Condition=Get bIsFalling())
   > Set AccScaleVals(AccScaleVals="0.1,0.1,0")
   Set AccScaleVals(execute=@n54.else, AccScaleVals="0.4,0.4,0.4")
   Set AccScaleVals(execute=@n53.else, AccScaleVals="0.4,0.4,0.4")
   Set FrontCloth Rigid alpha(execute=@n9.then_2, FrontCloth Rigid alpha:double=SelectFloat(A=0.8, B=1, bPickA=Get Is moving forward())) -> (Output_Get:double)
   > Set BackCloth Rigid alpha(BackCloth Rigid alpha:double=SelectFloat(A=0, B=1, bPickA=Get WearingCape() -> (WearingCape:bool))) -> (Output_Get:double)
   n55 = Gravity Override State Function DEPRECATED(
    execute=@n9.then_1
    CharacterPartAnimInstance=Self-Reference()
    GameTimeInSeconds=GetGameTimeInSeconds()
    GravityOverrideParams=Get Gravity Override Params()
   )
   > Set Gravity Override(Gravity Override=@n55)
   n56 = Get PawnMovementDirectionAzimuth() -> (PawnMovementDirectionAzimuth:float)
   n57 = "float > float"(A=@n56, B=-45)
   n58 = "float < float"(A=@n56, B=45)
   Set Is moving forward(execute=@n9.then_5, Is moving forward:bool=AND Boolean(A=AND Boolean(A=@n57, B=@n58), B=Get bIsMoving2D())) -> (Output_Get:bool)
   NOT Boolean(A=Get bIsCrouching())
   float * float(A=Get 30 FPS correction(), B=@n24)
   float * float(A=Get 30 FPS correction(), B=@n16)
   float * float(A=Get 30 FPS correction(), B=@n28)
   float * float(A=Get 30 FPS correction(), B=@n33)
   float * float(A=Get 30 FPS correction(), B=@n20)
  ]}
]
variables [
 " jump alpha" {type=number}
 30 FPS correction {type=number}
 AccScaleVals {type=Vector}
 Alpha_Thigh_L_Back {type=number}
 Alpha_Thigh_L_Front {type=number}
 Alpha_Thigh_L_Front_Fwd {type=number}
 Alpha_Thigh_L_Front_Fwd crouch {type=number}
 Alpha_Thigh_L_Front_Sprint {type=number}
 Alpha_Thigh_LBackCross {type=number}
 Alpha_Thigh_R_Back {type=number}
 Alpha_Thigh_R_Front {type=number}
 Alpha_Thigh_R_Front_Fwd {type=number}
 Alpha_Thigh_R_Front_Fwd crouch {type=number}
 Alpha_Thigh_R_Front_Sprint {type=number}
 bachcloth fwd {type=number}
 Back mvmnt corr {type=boolean}
 BackCloth Rigid alpha {type=number}
 Backward_Direct {type=number}
 calf_l_alpha {type=number}
 calf_r_alpha {type=number}
 Crouching to the left Alpha {type=number}
 Crouching to the right Alpha {type=number}
 Delta Time X {type=number}
 FrontCloth Rigid alpha {type=number}
 FWD Left {type=boolean}
 FWD Right {type=boolean}
 Gravity Override {type=Vector}
 Gravity Override Params {type=GravityOverrideParamsStruct}
 Heavy Weapon Crouch {type=number}
 Is Backbling Long {type=boolean}
 Is crouching and moving {type=boolean}
 Is Crouching and moving alpha {type=number}
 Is crouching and not moving {type=boolean}
 Is moving alpha {type=number}
 Is moving forward {type=boolean}
 isSprintTact {type=boolean}
 jump alpha Back {type=number}
 jump alpha Forward {type=number}
 jump alpha Left {type=number}
 jump alpha Right {type=number}
 Jumping to the Left Alpha {type=number}
 Jumping to the Right Alpha {type=number}
 Left Leg Bend Euler Z Alpha {type=number}
 Left Leg crouch Idle Bend Euler Z Alpha {type=number}
 Left Leg crouch moving Bend Euler Z Alpha {type=number}
 Left Leg Crouch Moving Forward or Backward Euler Z Alpha {type=number}
 Left mvmnt corr {type=boolean}
 Left_Direct {type=number}
 Local Noise Offset {type=Vector}
 "Local Noise Transform  Location" {type=Vector}
 Mat Instance Dynamic {type=MaterialInstanceDynamic}
 Moving Back L Front {type=number}
 Moving Back R Front {type=number}
 On Ground Anim Dynamics Alpha {type=number}
 Overlay Mat Instance Dynamic {type=MaterialInstanceDynamic}
 pants alpha dynamics {type=number}
 PlayerPawn {type=FortPlayerPawnAthena}
 R1 Velocity Transform Location {type=Vector}
 R1 Velocity Transform Velocity {type=Vector}
 R2 Velocity Transform Location {type=Vector}
 R2 Velocity Transform Velocity {type=Vector}
 Raw thigh_l {type=number}
 Raw thigh_R {type=number}
 RBANSprintAlpha {type=number}
 Right Leg Bend Euler Z Alpha {type=number}
 Right Leg crouch Bend Euler Z Alpha {type=number}
 Right Leg crouch Idle Bend Euler Z Alpha {type=number}
 Right Leg crouch moving Bend Euler Z Alpha {type=number}
 Right mvmnt corr {type=boolean}
 Right_Direct {type=number}
 SkelMesh {type=SkeletalMeshComponent}
 Start moving forward {type=number}
 Start Moving Fwd {type=number}
 Start moving left {type=number}
 Start moving right {type=number}
 Thigh_L_FWD {type=number}
 Thigh_R_FWD {type=number}
 upperarm_l alpha {type=number}
 upperarm_r alpha {type=number}
 WearingBack {type=boolean}
 WearingCape {type=boolean}
 WPO Pho Wind Intensity {type=number}
]
