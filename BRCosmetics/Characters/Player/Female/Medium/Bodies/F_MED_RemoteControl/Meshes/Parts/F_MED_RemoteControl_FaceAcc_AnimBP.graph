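# Anim Blueprint for the female RemoteControl cosmetic's face accessory mesh, adding secondary motion to hair strands, bangs and hair bones.
# It copies the attached parent's pose, then runs a serial chain of root-relative AnimDynamics nodes with per-bone limits, damping and some springs.
# Fixed additive bone rotations offset both bang master bones and both side strands.
# The EventGraph computes HeadLookUpAlpha from head versus spine_05, but the AnimGraph never reads it.
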AnimBlueprint /BRCosmetics/Characters/Player/Female/Medium/Bodies/F_MED_RemoteControl/Meshes/Parts/F_MED_RemoteControl_FaceAcc_AnimBP
target_skeleton=&/F_MED_RemoteControl_FaceAcc_Skeleton
graphs [
 AnimGraph {nodes [
   n1 = Anim Dynamics - Bone: dyn_hair_left_strand_A(ComponentPose=Local To Component(LocalPose=Copy Pose From Mesh() {bUseAttachedParent=true}), Alpha=1, ExternalForce="0,0,0") {
    angularDampingOverride=1
    angularSpringConstant=1500
    bAngularSpring=true
    bGravityOverrideInSimSpace=true
    bOverrideAngularDamping=true
    gravityScale=0.35
    simulationSpace=RootRelative
    alphaBoolBlend {blendOption=Linear}
    boundBone {boneName=dyn_hair_left_strand_A}
    componentAppliedLinearAccClamp=<3000 3000 3000>
    componentLinearAccScale=<0.3 0.3 0.3>
    componentLinearVelScale=<0.3 0.3 0.3>
    physicsBodyDefinitions [{
      sphereCollisionRadius=0
      boundBone {boneName=dyn_hair_left_strand_A}
      boxExtents=<10 15 15>
      constraintSetup {angularLimitsMax=<5 -5 20>, angularLimitsMin=<-5 -20 -25>, angularTarget=<0 0 0>}
      localJointOffset=<3 0 0>
     }]
   }
   n2 = Anim Dynamics - Bone: dyn_hair_left_strand_A_01(ComponentPose=@n1, Alpha=1, ExternalForce="0,0,0") {
    angularDampingOverride=1
    angularSpringConstant=1500
    bAngularSpring=true
    bGravityOverrideInSimSpace=true
    bOverrideAngularDamping=true
    gravityScale=0.35
    simulationSpace=RootRelative
    alphaBoolBlend {blendOption=Linear}
    boundBone {boneName=dyn_hair_left_strand_A_01}
    componentAppliedLinearAccClamp=<3000 3000 3000>
    componentLinearAccScale=<0.3 0.3 0.3>
    componentLinearVelScale=<0.3 0.3 0.3>
    physicsBodyDefinitions [{
      sphereCollisionRadius=0
      boundBone {boneName=dyn_hair_left_strand_A_01}
      boxExtents=<10 15 15>
      constraintSetup {angularLimitsMax=<5 -5 20>, angularLimitsMin=<-5 -20 -25>, angularTarget=<0 0 0>}
      localJointOffset=<3 0 0>
     }]
   }
   n3 = Anim Dynamics - Bone: dyn_hair_left_strand_A_02(ComponentPose=@n2, Alpha=1, ExternalForce="0,0,0") {
    angularDampingOverride=1
    angularSpringConstant=1500
    bAngularSpring=true
    bGravityOverrideInSimSpace=true
    bOverrideAngularDamping=true
    gravityScale=0.35
    simulationSpace=RootRelative
    alphaBoolBlend {blendOption=Linear}
    boundBone {boneName=dyn_hair_left_strand_A_02}
    componentAppliedLinearAccClamp=<3000 3000 3000>
    componentLinearAccScale=<0.3 0.3 0.3>
    componentLinearVelScale=<0.3 0.3 0.3>
    physicsBodyDefinitions [{
      sphereCollisionRadius=0
      boundBone {boneName=dyn_hair_left_strand_A_02}
      boxExtents=<10 15 15>
      constraintSetup {angularLimitsMax=<5 -5 10>, angularLimitsMin=<-5 -20 -25>, angularTarget=<0 0 0>}
      localJointOffset=<3 0 0>
     }]
   }
   n4 = Anim Dynamics - Bone: dyn_hair_right_strand_A(ComponentPose=@n3, Alpha=0.7, ExternalForce="0,0,0") {
    angularDampingOverride=1
    angularSpringConstant=1500
    bOverrideAngularDamping=true
    bUseGravityOverride=true
    gravityScale=0.35
    simulationSpace=RootRelative
    alphaBoolBlend {blendOption=Linear}
    boundBone {boneName=dyn_hair_right_strand_A}
    componentAppliedLinearAccClamp=<30000 30000 30000>
    componentLinearAccScale=<0.2 0.2 0.2>
    componentLinearVelScale=<0.2 0.2 0.2>
    gravityOverride {z=-980}
    physicsBodyDefinitions [{
      sphereCollisionRadius=0
      boundBone {boneName=dyn_hair_right_strand_A}
      boxExtents=<10 15 15>
      constraintSetup {angularTargetAxis=AxisY, angularLimitsMax=<0 -10 20>, angularLimitsMin=<0 -25 -40>, angularTarget=<0 -10 0>}
      localJointOffset=<-3 0 0>
     }]
   }
   n5 = Anim Dynamics - Bone: dyn_hair_right_strand_A_01(ComponentPose=@n4, Alpha=0.7, ExternalForce="0,0,0") {
    angularDampingOverride=1
    angularSpringConstant=1500
    bOverrideAngularDamping=true
    bUseGravityOverride=true
    gravityScale=0.35
    simulationSpace=RootRelative
    alphaBoolBlend {blendOption=Linear}
    boundBone {boneName=dyn_hair_right_strand_A_01}
    componentAppliedLinearAccClamp=<30000 30000 30000>
    componentLinearAccScale=<0.2 0.2 0.2>
    componentLinearVelScale=<0.2 0.2 0.2>
    gravityOverride {z=-980}
    physicsBodyDefinitions [{
      sphereCollisionRadius=0
      boundBone {boneName=dyn_hair_right_strand_A_01}
      boxExtents=<10 15 15>
      constraintSetup {angularTargetAxis=AxisY, angularLimitsMax=<0 -10 20>, angularLimitsMin=<0 -25 -40>, angularTarget=<0 -10 0>}
      localJointOffset=<-3 0 0>
     }]
   }
   n6 = Anim Dynamics - Bone: dyn_hair_right_strand_A_02(ComponentPose=@n5, Alpha=0.7, ExternalForce="0,0,0") {
    angularDampingOverride=0.9
    angularSpringConstant=1500
    bOverrideAngularDamping=true
    bUseGravityOverride=true
    gravityScale=0.35
    simulationSpace=RootRelative
    alphaBoolBlend {blendOption=Linear}
    boundBone {boneName=dyn_hair_right_strand_A_02}
    componentAppliedLinearAccClamp=<3000 3000 3000>
    componentLinearAccScale=<0.2 0.2 0.2>
    componentLinearVelScale=<0.2 0.2 0.2>
    gravityOverride {z=-980}
    physicsBodyDefinitions [{
      sphereCollisionRadius=0
      boundBone {boneName=dyn_hair_right_strand_A_02}
      boxExtents=<7 15 15>
      constraintSetup {angularTargetAxis=AxisY, angularLimitsMax=<0 0 5>, angularLimitsMin=<0 -10 -5>, angularTarget=<0 -10 0>}
      localJointOffset=<-1 0 0>
     }]
   }
   n7 = Anim Dynamics - Bone: dyn_Hair_01(ComponentPose=@n6, Alpha=1, ExternalForce="0,0,0") {
    angularDampingOverride=0.99
    angularSpringConstant=75
    bAngularSpring=true
    bGravityOverrideInSimSpace=true
    bOverrideAngularDamping=true
    bUseGravityOverride=true
    simulationSpace=RootRelative
    alphaBoolBlend {blendOption=Linear}
    boundBone {boneName=dyn_Hair_01}
    componentAppliedLinearAccClamp=<3500 3500 2500>
    componentLinearAccScale=<0.5 0.5 0.35>
    componentLinearVelScale=<0.55 0.55 0.4>
    gravityOverride {z=-980}
    physicsBodyDefinitions [{
      collisionType=CustomSphere
      sphereCollisionRadius=2
      boundBone {boneName=dyn_Hair_01}
      boxExtents=<17 17 17>
      constraintSetup {angularLimitsMax=<10 5 0>, angularLimitsMin=<-10 -5 -20>, angularTarget=<0 0 0>}
      localJointOffset=<0 -5 0>
     }]
    relativeSpaceBone {boneName=root}
   }
   n8 = Anim Dynamics - Bone: dyn_Hair_02(ComponentPose=@n7, Alpha=1, ExternalForce="0,0,0") {
    angularDampingOverride=0.99
    angularSpringConstant=75
    bAngularSpring=true
    bGravityOverrideInSimSpace=true
    bOverrideAngularDamping=true
    bUseGravityOverride=true
    simulationSpace=RootRelative
    alphaBoolBlend {blendOption=Linear}
    boundBone {boneName=dyn_Hair_02}
    componentAppliedLinearAccClamp=<3500 3500 2500>
    componentLinearAccScale=<0.5 0.5 0.35>
    componentLinearVelScale=<0.55 0.55 0.4>
    gravityOverride {z=-980}
    physicsBodyDefinitions [{
      collisionType=CustomSphere
      sphereCollisionRadius=2
      boundBone {boneName=dyn_Hair_02}
      boxExtents=<17 17 17>
      constraintSetup {angularLimitsMax=<7 2 0>, angularLimitsMin=<-10 -5 -20>, angularTarget=<0 0 0>}
      localJointOffset=<0 -5 0>
     }]
    relativeSpaceBone {boneName=root}
   }
   n9 = Anim Dynamics - Bone: dyn_Hair_03(ComponentPose=@n8, Alpha=1, ExternalForce="0,0,0") {
    angularDampingOverride=0.99
    angularSpringConstant=75
    bAngularSpring=true
    bGravityOverrideInSimSpace=true
    bOverrideAngularDamping=true
    bUseGravityOverride=true
    simulationSpace=RootRelative
    alphaBoolBlend {blendOption=Linear}
    boundBone {boneName=dyn_Hair_03}
    componentAppliedLinearAccClamp=<3500 3500 2500>
    componentLinearAccScale=<0.5 0.5 0.35>
    componentLinearVelScale=<0.55 0.55 0.4>
    gravityOverride {z=-980}
    physicsBodyDefinitions [{
      collisionType=CustomSphere
      sphereCollisionRadius=2
      boundBone {boneName=dyn_Hair_03}
      boxExtents=<17 17 17>
      constraintSetup {angularLimitsMax=<7 2 0>, angularLimitsMin=<-10 -5 -20>, angularTarget=<0 0 0>}
      localJointOffset=<0 -5 0>
     }]
    relativeSpaceBone {boneName=root}
   }
   n10 = Anim Dynamics - Bone: dyn_bang_l_01(ComponentPose=@n9, Alpha=1, ExternalForce="0,0,0") {
    angularDampingOverride=1
    angularSpringConstant=1500
    bAngularSpring=true
    bGravityOverrideInSimSpace=true
    bOverrideAngularDamping=true
    gravityScale=0.35
    simulationSpace=RootRelative
    alphaBoolBlend {blendOption=Linear}
    boundBone {boneName=dyn_bang_l_01}
    componentAppliedLinearAccClamp=<3000 3000 3000>
    componentLinearAccScale=<0.3 0.3 0.3>
    componentLinearVelScale=<0.3 0.3 0.3>
    physicsBodyDefinitions [{
      sphereCollisionRadius=0
      boundBone {boneName=dyn_bang_l_01}
      boxExtents=<10 15 15>
      constraintSetup {angularLimitsMax=<5 45 0>, angularLimitsMin=<-5 -30 -15>, angularTarget=<0 0 0>}
      localJointOffset=<3 0 0>
     }]
   }
   n11 = "Transform (Modify) Bone - Bone: MSTR_bang_l_01"(ComponentPose=@n10, Alpha=1, Translation="0,0,0", Rotation="15,0,0", Scale="1,1,1") {
    rotationMode=BMM_Additive
    rotationSpace=BCS_BoneSpace
    alphaBoolBlend {blendOption=Linear}
    boneToModify {boneName=MSTR_bang_l_01}
    rotation {pitch=15}
   }
   n12 = Anim Dynamics - Bone: dyn_bang_r_01(ComponentPose=@n11, Alpha=1, ExternalForce="0,0,0") {
    angularDampingOverride=1
    angularSpringConstant=1500
    bAngularSpring=true
    bGravityOverrideInSimSpace=true
    bOverrideAngularDamping=true
    gravityScale=0.35
    simulationSpace=RootRelative
    alphaBoolBlend {blendOption=Linear}
    boundBone {boneName=dyn_bang_r_01}
    componentAppliedLinearAccClamp=<3000 3000 3000>
    componentLinearAccScale=<0.3 0.3 0.3>
    componentLinearVelScale=<0.3 0.3 0.3>
    physicsBodyDefinitions [{
      sphereCollisionRadius=0
      boundBone {boneName=dyn_bang_r_01}
      boxExtents=<10 15 15>
      constraintSetup {angularLimitsMax=<5 15 0>, angularLimitsMin=<-5 -45 -15>, angularTarget=<0 0 0>}
      localJointOffset=<3 0 0>
     }]
   }
   n13 = "Transform (Modify) Bone - Bone: MSTR_bang_r_01"(ComponentPose=@n12, Alpha=1, Translation="0,0,0", Rotation="-15,0,0", Scale="1,1,1") {
    rotationMode=BMM_Additive
    rotationSpace=BCS_BoneSpace
    alphaBoolBlend {blendOption=Linear}
    boneToModify {boneName=MSTR_bang_r_01}
    rotation {pitch=-15}
   }
   n14 = "Transform (Modify) Bone - Bone: dyn_hair_left_strand_A"(ComponentPose=@n13, Alpha=1, Translation="0,0,0", Rotation="0,15,0", Scale="1,1,1") {
    rotationMode=BMM_Additive
    rotationSpace=BCS_BoneSpace
    alphaBoolBlend {blendOption=Linear}
    boneToModify {boneName=dyn_hair_left_strand_A}
    rotation {yaw=15}
   }
   n15 = "Transform (Modify) Bone - Bone: dyn_hair_left_strand_A_01"(ComponentPose=@n14, Alpha=1, Translation="0,0,0", Rotation="0,10,0", Scale="1,1,1") {
    rotationMode=BMM_Additive
    rotationSpace=BCS_BoneSpace
    alphaBoolBlend {blendOption=Linear}
    boneToModify {boneName=dyn_hair_left_strand_A_01}
    rotation {yaw=10}
   }
   n16 = "Transform (Modify) Bone - Bone: dyn_hair_left_strand_A_02"(ComponentPose=@n15, Alpha=1, Translation="0,0,0", Rotation="0,5,0", Scale="1,1,1") {
    rotationMode=BMM_Additive
    rotationSpace=BCS_BoneSpace
    alphaBoolBlend {blendOption=Linear}
    boneToModify {boneName=dyn_hair_left_strand_A_02}
    rotation {yaw=5}
   }
   n17 = "Transform (Modify) Bone - Bone: dyn_hair_right_strand_A"(ComponentPose=@n16, Alpha=1, Translation="0,0,0", Rotation="0,15,0", Scale="1,1,1") {
    rotationMode=BMM_Additive
    rotationSpace=BCS_BoneSpace
    alphaBoolBlend {blendOption=Linear}
    boneToModify {boneName=dyn_hair_right_strand_A}
    rotation {yaw=15}
   }
   n18 = "Transform (Modify) Bone - Bone: dyn_hair_right_strand_A_01"(ComponentPose=@n17, Alpha=1, Translation="0,0,0", Rotation="0,10,0", Scale="1,1,1") {
    rotationMode=BMM_Additive
    rotationSpace=BCS_BoneSpace
    alphaBoolBlend {blendOption=Linear}
    boneToModify {boneName=dyn_hair_right_strand_A_01}
    rotation {yaw=10}
   }
   n19 = "Transform (Modify) Bone - Bone: dyn_hair_right_strand_A_02"(ComponentPose=@n18, Alpha=1, Translation="0,0,0", Rotation="0,5,0", Scale="1,1,1") {
    rotationMode=BMM_Additive
    rotationSpace=BCS_BoneSpace
    alphaBoolBlend {blendOption=Linear}
    boneToModify {boneName=dyn_hair_right_strand_A_02}
    rotation {yaw=5}
   }
   Output Pose(Result=Component To Local(ComponentPose=@n19)) {layerGroup=None}
  ]}
 EventGraph {nodes [
   Event BlueprintUpdateAnimation()
   > Set HeadLookUpAlpha(HeadLookUpAlpha:float=GetAlphaFromReferenceTransformDelta(SkeletalMeshComponent=Get Skel Mesh(), BoneName=head, BaseName=spine_05, InputRange=(X=0,Y=-40), TwistAxis="1,0,0")) -> (Output_Get:float)
   Event BlueprintInitializeAnimation()
   > Set Skel Mesh(Skel Mesh=GetOwnerSkeletalMeshComponent())
  ]}
]
variables [
 HeadLookUpAlpha {type=number}
 Skel Mesh {type=SkeletalMeshComponent}
]
